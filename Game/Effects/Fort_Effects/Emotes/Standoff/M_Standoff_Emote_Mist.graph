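# Unlit translucent Niagara sprite material for the mist effect in the Standoff emote.
# An 8x8 falling-mist SubUV flipbook plus a faint tiling dust texture drives both brightness and opacity, scaled by particle color and alpha.
# Brightness gets a minimum floor, a tint and an emissive adjustment function.
# Opacity uses near-camera and depth fades by default, while mobile swaps both for a simple multiplier.
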Material /Game/Effects/Fort_Effects/Emotes/Standoff/M_Standoff_Emote_Mist
graph {
 properties {
  bUsedWithNanite=true
  bUsedWithNiagaraSprites=true
  blendMode=BLEND_Translucent
  pixelDepthOffsetMode=PDOM_Legacy
  shadingModel=MSM_Unlit
  translucencyLightingMode=TLM_VolumetricPerVertexNonDirectional
  usedShadingModels=Unlit
 }
 nodes [
  n1 = ParticleColor()
  n2 = TextureSampleParameterSubUV!(UVs, Apply View MipBias) {
   group=Texture
   parameterName=Cloud SubUV
   samplerType=SAMPLERTYPE_LinearColor
   texture=/Game/Effects/Fort_Effects/Textures/Smoke/T_FallingMist_8x8_01
  }
  n3 = TextureSampleParameter2D(UVs=TextureCoordinate() {uTiling=4, vTiling=4}) {group=Texture, parameterName=Tiling Dust Texture, sortPriority=2, texture=/Game/Global/Textures/Default/PointSplatter/T_Tiling_Dust}
  n4 = Multiply(A=@n3.RGB) {constB=0.35}
  n5 = Add(A=LinearInterpolate(A=ScalarParameter() {parameterName=SubUVMinBrightness}, Alpha=@n2.R), B=@n4)
  MaterialFunctionCall(Emissive=Multiply(
    A=Multiply(A=@n1.RGB, B=@n5)
    B=VectorParameter() {group=Color, parameterName=EmissiveTweak, sortPriority=1, defaultValue=<rgba 2 2 2 0>}.RGB
   ) {constB=0.5}) {materialFunction=&MF_FX_EmissiveAdjustments}
  n6 = Multiply(A=@n1.A, B=Add(A=@n2.B, B=@n4))
  n7 = MaterialFunctionCall(Distance=ScalarParameter() {defaultValue=500, group=Opacity, parameterName=Near Camera Fade Distance, sortPriority=2}) {materialFunction=&Near_Camera_Fade}
  ShadingPathSwitch(
   Default=DepthFade(
    Opacity=Multiply(A=@n6, B=@n7.Result)
    FadeDistance=ScalarParameter() {defaultValue=100, group=Opacity, parameterName=Depth Fade Distance, sortPriority=1}
   )
   Mobile=Multiply(A=ScalarParameter() {defaultValue=1, parameterName=Mobile Opacity Multiplier}, B=@n6)
  )
 ]
}
